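# Physics asset for the dynamic bandanna face accessory on the male medium Slither player body, simulating its two hanging tails.
# Left and right five-capsule chains descend from master bones linked to the head through free-swinging constraints.
# Chain joints keep side swing and twist tight while allowing a much wider second swing, so each tail bends mainly on one axis.
# Larger capsules on head, neck, upper spine and clavicles approximate the body around it.
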PhysicsAsset /BRCosmetics/Characters/Player/Male/Medium/Bodies/M_MED_Slither/Meshes/Parts/ASN/M_MED_Slither_ANS_FaceAcc_Dyn_bandanna
bodies {
 clavicle_l [{shapeType=Capsule, length=8.375, radius=8.125, center=<7.0524 -1.086 0.4439>, rotation=<rot 79.9998 0.0001 -2.53>}]
 clavicle_r [{shapeType=Capsule, length=8.375, radius=8.125, center=<-7.0524 1.086 -0.4439>, rotation=<rot 79.9998 0.0001 177.47>}]
 dyn_bandanna_MSTR [{shapeType=Capsule, length=0.5, radius=0.5}]
 dyn_bandanna_l_01 [{shapeType=Capsule, length=1.625, radius=1.5, shapeName=dyn_bandanna_l_01_capsule, center=<2 0 0>, rotation=<rot 90 0 0>}]
 dyn_bandanna_l_02 [{shapeType=Capsule, length=1.875, radius=1.625, shapeName=dyn_bandanna_l_02_capsule, center=<2 0 0>, rotation=<rot 90 0 0>}]
 dyn_bandanna_l_03 [{shapeType=Capsule, length=2.25, radius=1.75, shapeName=dyn_bandanna_l_03_capsule, center=<2 0 0>, rotation=<rot 90 0 0>}]
 dyn_bandanna_l_04 [{shapeType=Capsule, length=1.125, radius=1.875, shapeName=dyn_bandanna_l_04_capsule, center=<2 0 0>, rotation=<rot 0 0 90>}]
 dyn_bandanna_l_05 [{shapeType=Capsule, length=1.375, radius=1.875, shapeName=dyn_bandanna_l_05_capsule, center=<2 0 0>, rotation=<rot 0 0 90>}]
 dyn_bandanna_l_MSTR [{shapeType=Capsule, length=0.505, radius=0.505, shapeName=dyn_bandanna_l_MSTR_capsule, rotation=<rot 0 -90 0>}]
 dyn_bandanna_r_01 [{shapeType=Capsule, length=1.625, radius=1.5, shapeName=dyn_bandanna_r_01_capsule, center=<2 0 0>, rotation=<rot 90 0 0>}]
 dyn_bandanna_r_02 [{shapeType=Capsule, length=1.875, radius=1.625, shapeName=dyn_bandanna_r_02_capsule, center=<2 0 0>, rotation=<rot 90 0 0>}]
 dyn_bandanna_r_03 [{shapeType=Capsule, length=2.25, radius=1.75, shapeName=dyn_bandanna_r_03_capsule, center=<2 0 0>, rotation=<rot 90 0 0>}]
 dyn_bandanna_r_04 [{shapeType=Capsule, length=0.875, radius=1.875, shapeName=dyn_bandanna_r_04_capsule, center=<2 0 0>, rotation=<rot 0 0 90>}]
 dyn_bandanna_r_05 [{shapeType=Capsule, length=1.125, radius=1.875, shapeName=dyn_bandanna_r_05_capsule, center=<2 0 0>, rotation=<rot 0 0 90>}]
 dyn_bandanna_r_MSTR [{shapeType=Capsule, length=0.505, radius=0.505, shapeName=dyn_bandanna_r_MSTR_capsule, rotation=<rot 0 -90 0>}]
 head [{shapeType=Capsule, length=4.5, radius=6.75, center=<6.193 2.1057 0>, rotation=<rot 90 1.7899 3.9846>}]
 neck_01 [{shapeType=Capsule, length=4.75, radius=10.25, center=<3.7401 3.8929 0.0002>, rotation=<rot 90 178.2101 172.2765>}]
 spine_04 [{shapeType=Capsule, length=17.4328, radius=11.8515, center=<10.7594 0.7649 -0.0001>}]
 spine_05 [{shapeType=Capsule, length=13.4328, radius=12.1015, center=<2.0624 1.302 -0.0003>}]
}
constraints [
 {bone1Name=head, bone2Name=neck_01, swing1LimitDegrees=45, swing2LimitDegrees=45, twistLimitDegrees=45}
 {
  bone1Name=dyn_bandanna_MSTR
  bone2Name=head
  swing1LimitDegrees=45
  swing1Motion=Free
  swing2LimitDegrees=45
  swing2Motion=Free
  twistLimitDegrees=45
  twistMotion=Free
 }
 {bone1Name=dyn_bandanna_l_01, bone2Name=dyn_bandanna_l_MSTR, swing1LimitDegrees=10, swing2LimitDegrees=45, twistLimitDegrees=2}
 {bone1Name=dyn_bandanna_l_02, bone2Name=dyn_bandanna_l_01, swing1LimitDegrees=5, swing2LimitDegrees=45, twistLimitDegrees=2}
 {bone1Name=dyn_bandanna_l_03, bone2Name=dyn_bandanna_l_02, swing1LimitDegrees=5, swing2LimitDegrees=60, twistLimitDegrees=5}
 {bone1Name=dyn_bandanna_l_04, bone2Name=dyn_bandanna_l_03, swing1LimitDegrees=5, swing2LimitDegrees=60, twistLimitDegrees=5}
 {bone1Name=dyn_bandanna_r_01, bone2Name=dyn_bandanna_r_MSTR, swing1LimitDegrees=10, swing2LimitDegrees=45, twistLimitDegrees=2}
 {bone1Name=dyn_bandanna_r_02, bone2Name=dyn_bandanna_r_01, swing1LimitDegrees=5, swing2LimitDegrees=45, twistLimitDegrees=2}
 {bone1Name=dyn_bandanna_r_03, bone2Name=dyn_bandanna_r_02, swing1LimitDegrees=5, swing2LimitDegrees=60, twistLimitDegrees=5}
 {bone1Name=dyn_bandanna_r_04, bone2Name=dyn_bandanna_r_03, swing1LimitDegrees=5, swing2LimitDegrees=60, twistLimitDegrees=5}
 {bone1Name=dyn_bandanna_r_05, bone2Name=dyn_bandanna_r_04, swing1LimitDegrees=5, swing2LimitDegrees=60, twistLimitDegrees=5}
 {bone1Name=dyn_bandanna_l_05, bone2Name=dyn_bandanna_l_04, swing1LimitDegrees=5, swing2LimitDegrees=60, twistLimitDegrees=5}
 {
  bone1Name=neck_01
  bone2Name=spine_05
  swing1LimitDegrees=45
  swing1Motion=Free
  swing2LimitDegrees=45
  swing2Motion=Free
  twistLimitDegrees=45
  twistMotion=Free
 }
 {
  bone1Name=spine_05
  bone2Name=spine_04
  swing1LimitDegrees=45
  swing1Motion=Free
  swing2LimitDegrees=45
  swing2Motion=Free
  twistLimitDegrees=45
  twistMotion=Free
 }
 {
  bone1Name=clavicle_r
  bone2Name=spine_05
  swing1LimitDegrees=45
  swing1Motion=Free
  swing2LimitDegrees=45
  swing2Motion=Free
  twistLimitDegrees=45
  twistMotion=Free
 }
 {
  bone1Name=dyn_bandanna_l_MSTR
  bone2Name=dyn_bandanna_MSTR
  swing1LimitDegrees=45
  swing1Motion=Free
  swing2LimitDegrees=45
  swing2Motion=Free
  twistLimitDegrees=45
  twistMotion=Free
 }
 {
  bone1Name=dyn_bandanna_r_MSTR
  bone2Name=dyn_bandanna_MSTR
  swing1LimitDegrees=45
  swing1Motion=Free
  swing2LimitDegrees=45
  swing2Motion=Free
  twistLimitDegrees=45
  twistMotion=Free
 }
 {
  bone1Name=clavicle_l
  bone2Name=spine_05
  swing1LimitDegrees=45
  swing1Motion=Free
  swing2LimitDegrees=45
  swing2Motion=Free
  twistLimitDegrees=45
  twistMotion=Free
 }
]
