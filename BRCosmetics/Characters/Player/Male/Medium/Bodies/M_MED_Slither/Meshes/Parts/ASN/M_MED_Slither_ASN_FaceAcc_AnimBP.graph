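# Post-copy AnimBP for the Slither face accessory mesh, animating the left and right bandanna tails over the attached parent's pose.
# When bRigidBodyEnabled is set, Look At master bones feed a head-relative RigidBody simulation.
# Otherwise, Look At bones feed planar-limited Trail controllers weighted by movement-derived BwdMovement, plus front-end-only constraints and offsets.
# The event graph computes BwdMovement each update from PlayerMovement, remapped and clamped.
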AnimBlueprint /BRCosmetics/Characters/Player/Male/Medium/Bodies/M_MED_Slither/Meshes/Parts/ASN/M_MED_Slither_ASN_FaceAcc_AnimBP
target_skeleton=&/M_MED_Slither_ASN_FaceAcc_Skeleton
graphs [
 AnimGraph {nodes [
   n1 = Look At - Bone: dyn_bandanna_r_01(ComponentPose=Local To Component(LocalPose=Copy Pose From Mesh() {bUseAttachedParent=true}), Alpha=1) {
    alphaBoolBlend {blendOption=Linear}
    boneToModify {boneName=dyn_bandanna_r_01}
    lookAtLocation=<-20 -8 8>
    lookAtTarget {boneReference {boneName=neck_01}}
    lookAt_Axis {axis=<1 0>}
   }
   n2 = Look At - Bone: dyn_bandanna_r_02(ComponentPose=@n1, Alpha=1) {
    alphaBoolBlend {blendOption=Linear}
    boneToModify {boneName=dyn_bandanna_r_02}
    lookAtLocation=<-20 -30 8>
    lookAtTarget {boneReference {boneName=spine_05}}
    lookAt_Axis {axis=<1 0>}
   }
   n3 = Look At - Bone: dyn_bandanna_l_01(ComponentPose=@n2, Alpha=1) {
    alphaBoolBlend {blendOption=Linear}
    boneToModify {boneName=dyn_bandanna_l_01}
    lookAtLocation=<-20 -8 -8>
    lookAtTarget {boneReference {boneName=neck_01}}
    lookAt_Axis {axis=<1 0>}
   }
   n4 = Look At - Bone: dyn_bandanna_l_02(ComponentPose=@n3, Alpha=1) {
    alphaBoolBlend {blendOption=Linear}
    boneToModify {boneName=dyn_bandanna_l_02}
    lookAtLocation=<-20 -30 -8>
    lookAtTarget {boneReference {boneName=spine_05}}
    lookAt_Axis {axis=<1 0>}
   }
   n5 = Get BwdMovement() -> (BwdMovement:double)
   n6 = Trail controller - Bone: dyn_bandanna_l_05(ComponentPose=@n4, Alpha=@n5) {
    bUsePlanarLimit=true
    chainLength=4
    maxDeltaTime=0.0333
    alphaScaleBiasClamp {bInterpResult=true, interpSpeedIncreasing=5}
    planarLimits [{drivingBone {boneName=spine_05}, planeTransform {location=<0 -10 0>, rotation=<rot 0 20 -90>, scale=<1 1 1>}}]
    rotationLimits [{limitMax=<180 180 180>}, {limitMax=<180 180 180>}, {limitMax=<180 180 180>}, {limitMax=<180 180 180>}]
    rotationOffsets [<0 0 0>, <0 0 0>, <0 0 0>, <0 0 0>]
    trailBone {boneName=dyn_bandanna_l_05}
    trailRelaxationSpeed {editorCurveData {keys [
       {arriveTangent=-5.1804, interpMode=RCIM_Cubic, leaveTangent=-5.1803, tangentMode=RCTM_User, time=0, value=40}
       {arriveTangent=-64.0742, interpMode=RCIM_Cubic, leaveTangent=-64.0742, tangentMode=RCTM_User, time=1, value=10}
      ]}}
   }
   n7 = Trail controller - Bone: dyn_bandanna_r_05(ComponentPose=@n6, Alpha=@n5) {
    bUsePlanarLimit=true
    chainLength=5
    maxDeltaTime=0.0333
    alphaScaleBiasClamp {bInterpResult=true, interpSpeedIncreasing=5}
    planarLimits [{drivingBone {boneName=spine_05}, planeTransform {location=<0 -10 0>, rotation=<rot 0 20 -90>, scale=<1 1 1>}}]
    rotationLimits [{limitMax=<180 180 180>}, {limitMax=<180 180 180>}, {limitMax=<180 180 180>}, {limitMax=<180 180 180>}, {limitMax=<180 180 180>}]
    rotationOffsets [<0 0 0>, <0 0 0>, <0 0 0>, <0 0 0>, <0 0 0>]
    trailBone {boneName=dyn_bandanna_r_05}
    trailRelaxationSpeed {editorCurveData {keys [
       {arriveTangent=-5.1804, interpMode=RCIM_Cubic, leaveTangent=-5.1803, tangentMode=RCTM_User, time=0, value=30}
       {arriveTangent=-64.0742, interpMode=RCIM_Cubic, leaveTangent=-64.0742, tangentMode=RCTM_User, time=1, value=5}
      ]}}
   }
   n8 = Get bIsInFrontEnd()
   n9 = "Constraint - dyn_bandanna_r_04 (P)"!(ComponentPose=@n7, bAlphaBoolEnabled:bool=@n8, ConstraintWeights_0=1) -> (Pose) {
    alphaInputType=Bool
    alphaBoolBlend {blendOption=Linear}
    boneToModify {boneName=dyn_bandanna_r_04}
    constraintSetup [{targetBone {boneName=spine_05}}]
    constraintWeights [1]
   }
   n10 = "Constraint - dyn_bandanna_l_04 (P)"!(ComponentPose=@n9, bAlphaBoolEnabled:bool=@n8, ConstraintWeights_0=1) -> (Pose) {
    alphaInputType=Bool
    alphaBoolBlend {blendOption=Linear}
    boneToModify {boneName=dyn_bandanna_l_04}
    constraintSetup [{targetBone {boneName=spine_05}}]
    constraintWeights [1]
   }
   n11 = "Transform (Modify) Bone - Bone: dyn_bandanna_r_04"!(ComponentPose=@n10, bAlphaBoolEnabled:bool=@n8, Translation="-2.271,-4.615,-2.571", Rotation="0,0,0", Scale="1,1,1") -> (Pose) {
    alphaInputType=Bool
    translationMode=BMM_Additive
    translationSpace=BCS_BoneSpace
    alphaBoolBlend {blendOption=Linear}
    boneToModify {boneName=dyn_bandanna_r_04}
    translation=<-2.271 -4.615 -2.571>
   }
   n12 = "Transform (Modify) Bone - Bone: dyn_bandanna_l_04"!(ComponentPose=@n11, bAlphaBoolEnabled:bool=@n8, Translation="-1.632,-3.760,-2.436", Rotation="0,0,0", Scale="1,1,1") -> (Pose) {
    alphaInputType=Bool
    translationMode=BMM_Additive
    translationSpace=BCS_BoneSpace
    alphaBoolBlend {blendOption=Linear}
    boneToModify {boneName=dyn_bandanna_l_04}
    translation=<-1.632 -3.76 -2.436>
   }
   n13 = Look At - Bone: dyn_bandanna_r_MSTR(ComponentPose=Local To Component(LocalPose=Copy Pose From Mesh() {bUseAttachedParent=true}), Alpha=1) {
    alphaBoolBlend {blendOption=Linear}
    boneToModify {boneName=dyn_bandanna_r_MSTR}
    lookAtLocation=<-20 -12 10>
    lookAtTarget {boneReference {boneName=neck_01}}
    lookAt_Axis {axis=<1 0>}
   }
   n14 = Look At - Bone: dyn_bandanna_l_MSTR(ComponentPose=@n13, Alpha=1) {
    alphaBoolBlend {blendOption=Linear}
    boneToModify {boneName=dyn_bandanna_l_MSTR}
    lookAtLocation=<-20 -12 -7>
    lookAtTarget {boneReference {boneName=neck_01}}
    lookAt_Axis {axis=<1 0>}
   }
   n15 = RigidBody!(ComponentPose=@n14, Alpha, OverrideWorldGravity:Vector="0,0,-980", ExternalForce) -> (Pose) {
    bClampLinearTranslationLimitToRefPose=true
    bOverrideWorldGravity=true
    overridePhysicsAsset=&/M_MED_Slither_ANS_FaceAcc_Dyn_bandanna
    simulationSpace=BaseBoneSpace
    alphaBoolBlend {blendOption=Linear}
    baseBoneRef {boneName=head}
    componentAppliedLinearAccClamp=<2000 2000 2000>
    componentLinearAccScale=<0.3 0.3 0.2>
    componentLinearVelScale=<0.3 0.3 0.2>
    overrideWorldGravity {z=-980}
   }
   Output Pose(Result=Blend Poses by bool(BlendPose_0=Component To Local(ComponentPose=@n15), BlendPose_1=Component To Local(ComponentPose=@n12), bActiveValue=Get bRigidBodyEnabled()) {blendType=Linear}) {layerGroup=None}
  ]}
 EventGraph {nodes [
   Event BlueprintUpdateAnimation()
   > n1 = PlayerMovement(AnimInstance=Self-Reference(), XDotProductCoefficient=1, YDotProductCoefficient=1, ZDotProductCoefficient=1)
   n2 = MapRangeClamped(Value=@n1.YMovement, InRangeA=0, InRangeB=1)
   n3 = float * float(A="Absolute (Float)"(A=@n1), B=0.5)
   > Set BwdMovement(BwdMovement:double=MapRangeClamped(Value=float + float(A=@n3, B=@n2) {is=title K2Node_CommutativeAssociativeBinaryOperator}, InRangeA=0, InRangeB=1.1, OutRangeB=0.4)) -> (Output_Get:double)
  ]}
]
variables [BwdMovement {type=number}]
